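# Sound cue for a Donut interactable in the Season 12 frontend UI, reusing the reward screen chest loot upgrade sound.
# A single wave player feeds the output at reduced volume through the non-critical in-game UI sound class.
# A concurrency override allows one instance at a time, stopping the oldest when it retriggers.
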SoundCue /Game/Sounds/Fort_UI_Sounds/S12_Frontend/Donut_Interactable_Cue
root=WavePlayer_1
nodes [{id=WavePlayer_1, properties {soundWaveAssetPtr=/Game/Sounds/Fort_UI_Sounds/Rewards/Fort_RewardScreen_TeamScore_ChestLootUpgrade}}]
properties {
 bOverrideConcurrency=true
 maxDistance=2097152
 soundClassObject=&Fort_UI_InGame_NonCritical
 volumeMultiplier=0.4
 concurrencyOverrides {maxCount=1, resolutionRule=StopOldest}
}
