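# Looping ambience cue for the Season 12 frontend Armory, routed through the Fort_UI sound class at reduced volume.
# A single looping wave player feeds the output directly, with no randomization or mixing.
# Overridden attenuation turns off spatialization but keeps distance falloff and enables trace-based occlusion with low-pass filtering, so the bed plays unpanned yet still responds to listener position.
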SoundCue /Game/Sounds/Fort_UI_Sounds/S12_Frontend/UI_AMB_Armory_Cue
root=WavePlayer_0
nodes [{id=WavePlayer_0, properties {bLooping=true, soundWaveAssetPtr=&/AMB_S12_Armory_Loop}}]
properties {
 bOverrideAttenuation=true
 maxDistance=1548.7764
 soundClassObject=&Fort_UI
 volumeMultiplier=0.2
 attenuationOverrides {
  bEnableOcclusion=true
  bSpatialize=false
  falloffDistance=856.2488
  occlusionLowPassFilterFrequency=1000
  occlusionTraceChannel=ECC_GameTraceChannel11
  attenuationShapeExtents=<692.5275 222.7978 669.0881>
 }
 concurrencyOverrides {maxCount=1, resolutionRule=PreventNew}
}
