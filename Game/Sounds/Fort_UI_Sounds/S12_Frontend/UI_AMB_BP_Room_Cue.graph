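# Looping ambience cue for the Season 12 Battle Pass room in the frontend UI.
# A single looping Wave Player feeds the Battle Pass loop straight to the output at reduced volume, under the Fort_UI sound class.
# Overridden attenuation disables spatialization while defining a falloff distance, so the bed is not panned by position.
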SoundCue /Game/Sounds/Fort_UI_Sounds/S12_Frontend/UI_AMB_BP_Room_Cue
root=WavePlayer_0
nodes [{id=WavePlayer_0, properties {bLooping=true, soundWaveAssetPtr=&/AMB_S12_Battle_Pass_Loop}}]
properties {
 bOverrideAttenuation=true
 maxDistance=2030.6453
 soundClassObject=&Fort_UI
 volumeMultiplier=0.4
 attenuationOverrides {bSpatialize=false, falloffDistance=1111.4183, occlusionLowPassFilterFrequency=1000, attenuationShapeExtents {x=919.2269}}
 concurrencyOverrides {maxCount=1, resolutionRule=PreventNew}
}
