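# UI_AMB_Bathroom_Cue is a Sound Cue for a quiet, looping bathroom ambience in the Season 12 frontend UI.
# A single looping Wave Player feeds the output directly, routed to the Fort_Music_Emote_1P_PSM sound class.
# Inline attenuation fades it with distance without spatialized panning, and occlusion on a custom trace channel muffles and lowers it.
# The UI_Duck_StopOldest_Conc set governs concurrency, as the inline concurrency settings are not enabled.
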SoundCue /Game/Sounds/Fort_UI_Sounds/S12_Frontend/UI_AMB_Bathroom_Cue
root=WavePlayer_0
nodes [{id=WavePlayer_0, properties {bLooping=true, soundWaveAssetPtr=&/AMB_S12_Bathroom_Loop}}]
properties {
 bOverrideAttenuation=true
 maxDistance=1116.2954
 priority=0.5
 soundClassObject=/Game/Sounds/FortSoundClassesAndModes/Fort_Music_Emote_1P_PSM
 volumeMultiplier=0.2
 attenuationOverrides {
  bEnableOcclusion=true
  bSpatialize=false
  falloffDistance=791.2677
  occlusionLowPassFilterFrequency=1000
  occlusionTraceChannel=ECC_GameTraceChannel11
  occlusionVolumeAttenuation=0.4
  attenuationShapeExtents=<325.0277 222.7978 669.0881>
 }
 concurrencyOverrides {maxCount=1, resolutionRule=PreventNew}
 concurrencySet [&/UI_Duck_StopOldest_Conc]
}
