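# Sound cue for hovering over text in the Battle Pass frontend UI.
# A single wave player plays the Season 13 Battle Pass main hover sound at reduced volume, routed through the Fort_UI sound class.
# Attenuation is overridden to off, so it plays non-spatially.
# It references the shared frontend text hover concurrency asset, alongside override settings allowing one instance that stops the oldest.
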SoundCue /Game/Sounds/Fort_UI_Sounds/S12_Frontend/UI_BattlePass_Text_Hover_Cue
root=WavePlayer_4
nodes [{id=WavePlayer_4, properties {soundWaveAssetPtr=/Game/Sounds/Fort_UI_Sounds/S13_Frontend/S13_S13_Battlepass_Main_Hover}}]
properties {
 bOverrideAttenuation=true
 maxDistance=2097152
 soundClassObject=&Fort_UI
 volumeMultiplier=0.55
 attenuationOverrides {bAttenuate=false}
 concurrencyOverrides {maxCount=1, resolutionRule=StopOldest}
 concurrencySet [/Game/Sounds/Fort_UI_Sounds/S13_Frontend/UI_Frontend_Text_Hover_Conc]
}
