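# UI sound cue that plays the Battle Pass door-open sound in the Season 12 frontend.
# A single wave player feeds the root at reduced volume under the Fort_UI sound class.
# Overridden attenuation turns off distance falloff and spatialization, so playback is non-positional.
# A shared concurrency rule set handles voice limiting, as the stricter single-instance override values are present but not enabled.
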SoundCue /Game/Sounds/Fort_UI_Sounds/S12_Frontend/UI_Battle_Pass_Door_Open_Cue
root=WavePlayer_0
nodes [{id=WavePlayer_0, properties {soundWaveAssetPtr=&/UI_Battle_Pass_Door_Open}}]
properties {
 bOverrideAttenuation=true
 maxDistance=2097152
 soundClassObject=&Fort_UI
 volumeMultiplier=0.35
 attenuationOverrides {bAttenuate=false, bSpatialize=false, stereoSpread=5000}
 concurrencyOverrides {maxCount=1, resolutionRule=PreventNew}
 concurrencySet [/Game/Sounds/Concurrency_Rules/ThreeMaximum_StopFurthest_VolumeScale_02]
}
