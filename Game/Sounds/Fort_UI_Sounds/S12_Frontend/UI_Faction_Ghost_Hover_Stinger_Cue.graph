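# UI sound cue for the Ghost faction hover stinger in the Season 12 frontend.
# A single wave player plays the Ghost hover stinger wave directly, with no randomization or modulation.
# The cue routes through the Fort_UI sound class and uses a dedicated hover stinger concurrency set to govern overlapping plays.
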SoundCue /Game/Sounds/Fort_UI_Sounds/S12_Frontend/UI_Faction_Ghost_Hover_Stinger_Cue
root=WavePlayer_0
nodes [{id=WavePlayer_0, properties {soundWaveAssetPtr=&/UI_Faction_Ghost_Hover_Stinger}}]
properties {
 maxDistance=2097152
 soundClassObject=&Fort_UI
 volumeMultiplier=1
 concurrencySet [&/UI_Faction_Hover_Ghost_Stinger_Conc]
}
