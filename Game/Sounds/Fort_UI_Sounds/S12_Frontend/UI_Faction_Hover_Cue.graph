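# UI sound cue for hovering over a faction in the Season 12 frontend.
# It plays the reused Infiltration perk hover wave through a modulator that adds slight random pitch variation.
# It routes to the UI sound class at low volume, with a concurrency override that allows one instance and stops the oldest, so rapid hovers do not stack.
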SoundCue /Game/Sounds/Fort_UI_Sounds/S12_Frontend/UI_Faction_Hover_Cue
root=Modulator_0
nodes [
 {id=Modulator_0, children [WavePlayer_1], properties {pitchMax=1.01, pitchMin=0.98}}
 {id=WavePlayer_1, properties {soundWaveAssetPtr=/Game/Sounds/LTM/Infiltration/Infiltration_UI_Perk_Hover}}
]
properties {
 bOverrideConcurrency=true
 maxDistance=2097152
 soundClassObject=&Fort_UI
 volumeMultiplier=0.1
 concurrencyOverrides {maxCount=1, resolutionRule=StopOldest}
}
