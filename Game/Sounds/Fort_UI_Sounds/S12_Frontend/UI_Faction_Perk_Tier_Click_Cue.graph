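# Sound cue for the frontend faction perk tier click, routed through the Fort_UI sound class.
# A single wave player reuses an elevator click sample, feeding a modulator that adds slight random pitch variation.
# The cue raises the overall pitch and lowers volume to reshape the borrowed sample.
# A concurrency override allows one instance at a time, stopping the oldest when retriggered.
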SoundCue /Game/Sounds/Fort_UI_Sounds/S12_Frontend/UI_Faction_Perk_Tier_Click_Cue
root=Modulator_0
nodes [
 {id=Modulator_0, children [WavePlayer_0], properties {pitchMax=1.02, pitchMin=0.98}}
 {id=WavePlayer_0, properties {soundWaveAssetPtr=&/UI_Acces_Elevator_Click}}
]
properties {
 bOverrideConcurrency=true
 maxDistance=2097152
 pitchMultiplier=1.65
 soundClassObject=&Fort_UI
 volumeMultiplier=0.4
 concurrencyOverrides {maxCount=1, resolutionRule=StopOldest}
}
