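# Empty SoundCue for the S12 frontend faction select UI, named as a silent variant, so it produces no audio.
# It has no sound nodes and no root, leaving nothing to play.
# It is assigned to the Fort_UI sound class with a volume multiplier and a very large max distance.
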SoundCue /Game/Sounds/Fort_UI_Sounds/S12_Frontend/UI_Faction_Select_Silent_Cue
root=null
nodes []
properties {maxDistance=2097152, soundClassObject=&Fort_UI, volumeMultiplier=1}
