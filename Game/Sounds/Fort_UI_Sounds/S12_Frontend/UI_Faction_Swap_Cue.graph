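# UI sound cue for a faction swap, placed with the Season 12 frontend sounds and routed to the Fort_UI_InGame_Critical sound class.
# A single Wave Player plays the Infiltration LTM reroll click, reused here rather than a dedicated wave.
# An Enveloper at the root shapes it, holding full volume briefly then fading to silence in under a second, with the cue's overall volume reduced.
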SoundCue /Game/Sounds/Fort_UI_Sounds/S12_Frontend/UI_Faction_Swap_Cue
root=Enveloper_0
nodes [
 {
  id=Enveloper_0
  children [WavePlayer_0]
  properties {
   pitchCurve {editorCurveData {keys [{time=0}]}}
   volumeCurve {editorCurveData {keys [<at 0 1>, {arriveTangent=-1.25, interpMode=RCIM_Cubic, leaveTangent=-1.25, tangentMode=RCTM_Break, time=0.5, value=1}, <at 0.8 0>]}}
  }
 }
 {id=WavePlayer_0, properties {soundWaveAssetPtr=/Game/Sounds/LTM/Infiltration/Infiltration_UI_Reroll_Click}}
]
properties {maxDistance=2097152, soundClassObject=&Fort_UI_InGame_Critical, volumeMultiplier=0.55}
