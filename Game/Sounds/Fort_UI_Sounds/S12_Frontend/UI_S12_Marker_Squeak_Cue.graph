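# Sound cue for a marker squeak, named and filed as a Season 12 frontend UI sound.
# A single wave player feeds the squeak wave straight to the output, with no randomization or modulation and slightly reduced volume.
# It is routed to the loot-collect sound class and references a consumable ambient attenuation asset rather than UI-specific settings.
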SoundCue /Game/Sounds/Fort_UI_Sounds/S12_Frontend/UI_S12_Marker_Squeak_Cue
root=WavePlayer_0
nodes [{id=WavePlayer_0, properties {soundWaveAssetPtr=&/UI_S12_Marker_Squeak}}]
properties {
 attenuationSettings=/Game/Sounds/Attenuation/Fort_Consumable_Amb_Att
 maxDistance=2150
 soundClassObject=&Fort_Loot_Collect
 volumeMultiplier=0.85
 attenuationOverrides {bAttenuate=false, falloffDistance=2500}
}
