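# UI hover sound cue for the Shadow Operation tech element in the Season 12 frontend.
# A single wave player plays the Shadow faction hover wave through a modulator that adds slight random pitch variation.
# It routes to the Fort_UI sound class at reduced volume, with a shared faction-hover concurrency setting governing overlapping plays.
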SoundCue /Game/Sounds/Fort_UI_Sounds/S12_Frontend/UI_Shadow_Operation_Tech_Hover_Cue
root=Modulator_2
nodes [
 {id=Modulator_2, children [WavePlayer_2], properties {pitchMax=1.01, pitchMin=0.99}}
 {id=WavePlayer_2, properties {soundWaveAssetPtr=&/UI_Faction_Shadow_Hover}}
]
properties {
 maxDistance=2097152
 soundClassObject=&Fort_UI
 volumeMultiplier=0.8
 concurrencySet [&/UI_Faction_Hover_Conc, null]
}
